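# Niagara Fluids template system with one GPU emitter that simulates sand as position-based dynamics particles.
# A one-time burst fills a box of grains inside a grid volume placed at the owner's transform.
# Each update integrates gravity with a capped timestep and registers particles in a neighbor-query grid with cells one grain wide.
# The emitter defines friction values and a self-bound attribute reader, and has mesh and sprite renderers.
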NiagaraSystem /NiagaraFluids/Templates/Sand/Systems/SandSystem
dependencies {
 usedDataInterfaces [/Script/Niagara.NiagaraDataInterfaceNeighborQuery, &NiagaraDataInterfaceParticleRead]
 usedDynamicInputs []
 usedModules [/Niagara/Modules/System/SystemLifeCycle, &SystemState]
}
emitters {SandEmitter {
  simTarget=GPUComputeSim
  emitterSpawnScript [
   /SetVariables(Emitter.KineticFriction=0.2, Emitter.ParticleDiameter=10, Emitter.StaticFriction=0.25) {isSetParameters=true}
   /SetVariables(Emitter.UsePersistentIDs=true) {isSetParameters=true}
   /Niagara/Modules/NeighborQuery/NeighborQuery_SetResolution(
    NeighborQuery:NiagaraDataInterfaceNeighborQuery=$Emitter.NQ
    Resolution Method:ENiagaraGrid2DResolution=::NewEnumerator2
    WorldCellSize:NiagaraFloat=$Emitter.ParticleDiameter
    WorldGridExtents:Vector3f=<1000 1000 500>
   )
   /NiagaraFluids/Modules/Grid3D/Grid3D_CreateUnitToWorldTransform(
    LocalPivot:Vector3f=<0 0 0.5>
    Offset:NiagaraPosition=$Engine.Owner.Position
    Rotation:Quat4f=$Engine.Owner.Rotation
    WorldGridExtents:Vector3f=$Output.NeighborQuery_SetResolution.WorldGridExtents
   )
   /SetVariables(Emitter.AttributeReader=DataInterface() {propertyValues {EmitterBinding {bindingMode=Self, emitterName=SandEmitter}}}) {isSetParameters=true}
   /SetVariables() {isSetParameters=true}
  ]
  emitterUpdateScript [
   &EmitterState(Life Cycle Mode=::NewEnumerator1, Loop Behavior=::NewEnumerator1, Loop Duration Mode=::NewEnumerator1, UseLoopDelay=false)
   &SpawnBurst_Instantaneous(Spawn Count=15000, Use Spawn Probability=false)
   /SetVariables(Emitter.dt:NiagaraFloat=/Niagara/DynamicInputs/Math/Min_Float(A:NiagaraFloat=0.0333, B:NiagaraFloat=$Engine.DeltaTime)) {isSetParameters=true}
  ]
  particleSpawnScript [
   /NiagaraFluids/Modules/Particles/PBD_EmitBox(Mass=1, RandOffset=4, SizeX=40, SizeY=40, Spacing=0.1, WorldOffset:Vector3f=$Emitter.Grid3D_CreateUnitToWorldTransform.GridCenterPosition) {module=EmitBox}
   /SetVariables(
    Particles.Scale=&VectorFromFloat(Value=&Divide_Float(A=$Emitter.ParticleDiameter, B=100))
    Particles.SpriteSize=&Vector2DFromFloat(Value=$Emitter.ParticleDiameter)
   ) {isSetParameters=true}
  ]
  particleUpdateScript [
   /NiagaraFluids/Modules/Particles/PBD_Integrate(dt:NiagaraFloat=$Emitter.dt, g:Vector3f=<0 0 -980>) {module=PBDIntegrate}
   /Niagara/Modules/NeighborQuery/AddParticleToNeighborQuery(
    Neighbor Query:NiagaraDataInterfaceNeighborQuery=$Emitter.NQ
    Use Persistent IDs:NiagaraBool=$Emitter.UsePersistentIDs
    WorldToUnit:NiagaraMatrix=$Emitter.WorldToUnit
   ) {module=AddParticleToNeighborQuery001}
  ]
  rendererClasses [&NiagaraMeshRendererProperties, &NiagaraSpriteRendererProperties]
  renderers [
   {rendererClass=&NiagaraMeshRendererProperties, rendererIndex=0}
   {rendererClass=&NiagaraSpriteRendererProperties, rendererIndex=1}
  ]
 }}
summary {
 systemName=SandSystem
 emitters [{
   emitterName=SandEmitter
   simTarget=GPUComputeSim
   rendererClasses [&NiagaraMeshRendererProperties, &NiagaraSpriteRendererProperties]
  }]
 userVariables []
}
